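# Mono MetaSound Source for Tidal Crane single movement audio, using a swarmer attenuation asset and a concurrency cap.
# On play, it picks a random wave from an input array and scales the player's output by MasterVolume.
# The Default page also randomizes pitch and start time within the chosen wave's duration.
# The Mobile page omits both randomizations, keeping only the random wave pick and volume multiply.
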MetaSoundSource /ProtoSwarm/Assets/Audio/MSS_TidalCrane_Movement_Single
inputs [
 MasterVolume {type=Float}
 Single.Movement.Wave {type=WaveAsset:Array}
 UE.Source.OnPlay {type=Trigger}
]
interfaces [UE.OutputFormat.Mono:1.0, UE.Source:1.1]
outputs ["UE.OutputFormat.Mono.Audio:0" {type=Audio}]
pages [
 Default {
  is_build_graph=true
  input_defaults {MasterVolume=16.8, UE.Source.OnPlay=null, Single.Movement.Wave []}
  nodes [
   ArrayRandomGet_0 = Array.Random Get:WaveAsset:Array(In Array=^Single.Movement.Wave, Next=^UE.Source.OnPlay)
   GetWaveDuration_0 = "UE.Get Wave Duration: "(Wave=@ArrayRandomGet_0.Value)
   WavePlayer_0 = UE.Wave Player:Mono(
    Play=@ArrayRandomGet_0.On Next
    Wave Asset=@ArrayRandomGet_0.Value
    Pitch Shift=UE.RandomFloat(Next=^UE.Source.OnPlay) {also {defaults {Max=2, Min=-2}}}.Value
    Start Time=UE.RandomTime(Next=^UE.Source.OnPlay, Max=@GetWaveDuration_0.Duration).Value
   ) {also {defaults {Loop="<MetaSoundBuilderResult.SUCCEEDED: 0>"}}}
   Multiply_0 = UE.Multiply:Audio by Float(AdditionalOperands=^MasterVolume, PrimaryOperand=@WavePlayer_0.Out Mono)
  ]
  links [{dst="@UE.OutputFormat.Mono.Audio:0", src=Multiply_0.Out}]
 }
 Mobile {
  nodes [
   ArrayRandomGet_0 = Array.Random Get:WaveAsset:Array(In Array=^Single.Movement.Wave, Next=^UE.Source.OnPlay)
   WavePlayer_0 = UE.Wave Player:Mono(Play=@ArrayRandomGet_0.On Next, Wave Asset=@ArrayRandomGet_0.Value) {also {defaults {Loop="<MetaSoundBuilderResult.SUCCEEDED: 0>"}}}
   Multiply_0 = UE.Multiply:Audio by Float(AdditionalOperands=^MasterVolume, PrimaryOperand=@WavePlayer_0.Out Mono)
  ]
  links [{dst="@UE.OutputFormat.Mono.Audio:0", src=Multiply_0.Out}]
 }
]
properties {
 assetImportData=&self:AssetImportData
 attenuationSettings=/ProtoSwarm/Assets/Audio/TidalCrane_Swarmer/ATT_TidalCrane_Movement_Swarmer_Single
 bLooping=true
 duration=10000
 maxDistance=4200
 sampleRate=48000
 concurrencyOverrides {maxCount=4}
}
