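# MetaSound Source preset that plays looping movement audio for a single TidalCrane Swarmer creature.
# It inherits its graph from MSS_TidalCrane_Movement_Single, supplying three cluster loop waves and a master volume, with mono output.
# Playback uses swarmer-specific attenuation and the creature footsteps sound class, with voices limited through a shared swarmer movement concurrency set.
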MetaSoundSource /ProtoSwarm/Assets/Audio/TidalCrane_Swarmer/MSS_TidalCrane_Movement_Swarmer_Single
preset_parent=/ProtoSwarm/Assets/Audio/MSS_TidalCrane_Movement_Single
inputs [
 MasterVolume {type=Float}
 Single.Movement.Wave {type=WaveAsset:Array}
 UE.Source.OnPlay {type=Trigger}
]
interfaces [UE.OutputFormat.Mono:1.0, UE.Source:1.1]
outputs ["UE.OutputFormat.Mono.Audio:0" {type=Audio}]
pages [Default {input_defaults {
   MasterVolume=8
   UE.Source.OnPlay=null
   Single.Movement.Wave [
    /ProtoSwarm/Assets/Audio/TidalCrane_Swarmer/SoundWaves/Swarmer_Movement_Cluster_1_Loop_01
    /ProtoSwarm/Assets/Audio/TidalCrane_Swarmer/SoundWaves/Swarmer_Movement_Cluster_1_Loop_02
    /ProtoSwarm/Assets/Audio/TidalCrane_Swarmer/SoundWaves/Swarmer_Movement_Cluster_1_Loop_03
   ]
  }}]
properties {
 assetImportData=&self:AssetImportData
 attenuationSettings=&/ATT_TidalCrane_Movement_Swarmer_Single
 bLooping=true
 duration=10000
 maxDistance=4200
 sampleRate=48000
 soundClassObject=/Game/Sounds/FortSoundClassesAndModes/Fort_Creatures_Steps_Normal
 concurrencyOverrides {maxCount=4}
 concurrencySet [/ProtoSwarm/Assets/Audio/Settings/CONC_TidalCrane_Swarmer_Movement_Shared]
}
