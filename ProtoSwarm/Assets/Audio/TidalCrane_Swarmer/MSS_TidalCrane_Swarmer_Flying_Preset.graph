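# MetaSound Source preset that plays looping wing-flap movement audio for the flying TidalCrane Swarmer creature.
# It inherits its graph from MSS_TidalCrane_Movement_Single, supplying four wing-flap loop waves and a master volume as input defaults.
# The mono output uses swarmer flight attenuation and the creature foley sound class.
# A concurrency override tightly caps simultaneous instances, limiting how many flight loops play at once.
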MetaSoundSource /ProtoSwarm/Assets/Audio/TidalCrane_Swarmer/MSS_TidalCrane_Swarmer_Flying_Preset
preset_parent=/ProtoSwarm/Assets/Audio/MSS_TidalCrane_Movement_Single
inputs [
 MasterVolume {type=Float}
 Single.Movement.Wave {type=WaveAsset:Array}
 UE.Source.OnPlay {type=Trigger}
]
interfaces [UE.OutputFormat.Mono:1.0, UE.Source:1.1]
outputs ["UE.OutputFormat.Mono.Audio:0" {type=Audio}]
pages [Default {input_defaults {
   MasterVolume=1.5
   UE.Source.OnPlay=null
   Single.Movement.Wave [
    /ProtoSwarm/Assets/Audio/TidalCrane_Swarmer/SoundWaves/Swarmer_WingFlap_Loop_01
    /ProtoSwarm/Assets/Audio/TidalCrane_Swarmer/SoundWaves/Swarmer_WingFlap_Loop_02
    /ProtoSwarm/Assets/Audio/TidalCrane_Swarmer/SoundWaves/Swarmer_WingFlap_Loop_03
    /ProtoSwarm/Assets/Audio/TidalCrane_Swarmer/SoundWaves/Swarmer_WingFlap_Loop_04
   ]
  }}]
properties {
 assetImportData=&self:AssetImportData
 attenuationSettings=&/ATT_TidalCrane_Movement_Swarmer_Flight
 bLooping=true
 bOverrideConcurrency=true
 duration=10000
 maxDistance=9200
 sampleRate=48000
 soundClassObject=/Game/Sounds/FortSoundClassesAndModes/Fort_Creatures_Foley
 volume=0.85
 concurrencyOverrides {maxCount=2}
}
