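# One-shot mono MetaSound source for TidalCrane Swarmer creature vocalizations in ProtoSwarm.
# On play, it picks a random wave from an array of Swarmer VOX clips and plays it with a random pitch shift spread around a base value.
# The output is scaled by a master volume, and the sound finishes when the wave player does.
# It uses Swarmer VO attenuation, shared concurrency and the creature vocalization sound class.
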MetaSoundSource /ProtoSwarm/Assets/Audio/TidalCrane_Swarmer/MSS_TidalCrane_Swarmer_VO
inputs [
 MasterVolume {type=Float}
 Random.Pitch.Base {type=Float}
 Random.Pitch.Spread {type=Float}
 Sounds {type=WaveAsset:Array}
 UE.Source.OnPlay {type=Trigger}
]
interfaces [UE.OutputFormat.Mono:1.0, UE.Source.OneShot:1.0, UE.Source:1.1]
outputs [
 "UE.OutputFormat.Mono.Audio:0" {type=Audio}
 UE.Source.OneShot.OnFinished {type=Trigger}
]
pages [Default {
  input_defaults {
   MasterVolume=1.88
   Random.Pitch.Base=0
   Random.Pitch.Spread=2
   UE.Source.OnPlay=null
   Sounds [
    /ProtoSwarm/Assets/Audio/TidalCrane_Swarmer/SoundWaves/Swarmer_VOX_01
    /ProtoSwarm/Assets/Audio/TidalCrane_Swarmer/SoundWaves/Swarmer_VOX_03
    /ProtoSwarm/Assets/Audio/TidalCrane_Swarmer/SoundWaves/Swarmer_VOX_05
    /ProtoSwarm/Assets/Audio/TidalCrane_Swarmer/SoundWaves/Swarmer_VOX_08
    /ProtoSwarm/Assets/Audio/TidalCrane_Swarmer/SoundWaves/Swarmer_VOX_09
    /ProtoSwarm/Assets/Audio/TidalCrane_Swarmer/SoundWaves/Swarmer_VOX_12
    /ProtoSwarm/Assets/Audio/TidalCrane_Swarmer/SoundWaves/Swarmer_VOX_15
    /ProtoSwarm/Assets/Audio/TidalCrane_Swarmer/SoundWaves/Swarmer_VOX_16
    /ProtoSwarm/Assets/Audio/TidalCrane_Swarmer/SoundWaves/Swarmer_VOX_19
    /ProtoSwarm/Assets/Audio/TidalCrane_Swarmer/SoundWaves/Swarmer_VOX_20
    /ProtoSwarm/Assets/Audio/TidalCrane_Swarmer/SoundWaves/Swarmer_VOX_23
    /ProtoSwarm/Assets/Audio/TidalCrane_Swarmer/SoundWaves/Swarmer_VOX_26
    /ProtoSwarm/Assets/Audio/TidalCrane_Swarmer/SoundWaves/Swarmer_VOX_29
    /ProtoSwarm/Assets/Audio/TidalCrane_Swarmer/SoundWaves/Swarmer_VOX_30
    /ProtoSwarm/Assets/Audio/TidalCrane_Swarmer/SoundWaves/Swarmer_VOX_33
    /ProtoSwarm/Assets/Audio/TidalCrane_Swarmer/SoundWaves/Swarmer_VOX_34
   ]
  }
  nodes [
   ArrayRandomGet_0 = Array.Random Get:WaveAsset:Array(In Array=^Sounds, Next=^UE.Source.OnPlay) {also {defaults {Weights []}}}
   WavePlayer_0 = UE.Wave Player:Mono(
    Play=@ArrayRandomGet_0.On Next
    Wave Asset=@ArrayRandomGet_0.Value
    Pitch Shift=UE.RandomFloat(
     Next=^UE.Source.OnPlay
     Max=UE.Add:Float(PrimaryOperand=^Random.Pitch.Base, AdditionalOperands=^Random.Pitch.Spread).Out
     Min=UE.Subtract:Float(PrimaryOperand=^Random.Pitch.Base, AdditionalOperands=^Random.Pitch.Spread).Out
    ).Value
   )
   Multiply_0 = UE.Multiply:Audio by Float(AdditionalOperands=^MasterVolume, PrimaryOperand=@WavePlayer_0.Out Mono)
  ]
  links [{dst="@UE.OutputFormat.Mono.Audio:0", src=Multiply_0.Out}, {dst="@UE.Source.OneShot.OnFinished", src=WavePlayer_0.On Finished}]
 }]
properties {
 assetImportData=&self:AssetImportData
 attenuationSettings=/ProtoSwarm/Assets/Audio/Settings/ATT_TidalCrane_Swarmer_VO
 maxDistance=3000
 sampleRate=48000
 soundClassObject=/Game/Sounds/FortSoundClassesAndModes/Fort_Creatures_Vocalization_Normal
 concurrencySet [/ProtoSwarm/Assets/Audio/Settings/CONC_TidalCrane_Swarmer_VO_Shared]
}
