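# Sound cue for the ProtoSwarm Swarmer bug's death, layering a death sound with a quieter crystal burst.
# A quality switch gives the first two levels ten random death waves and the last two a three-wave subset, while the burst picks from four waves.
# Each layer has its own pitch modulation.
# Burst attenuation shapes falloff, and a concurrency override caps simultaneous voices despite the assigned global and local concurrency assets.
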SoundCue /ProtoSwarm/Assets/Audio/TidalCrane_Swarmer/SC_Bug_Swarmer_Death
root=Mixer_0
nodes [
 {id=Mixer_0, children [Modulator_0,Modulator_1], properties {inputVolume [1.5,0.5]}}
 {id=Modulator_0, children [QualityLevel_0], properties {pitchMax=1, pitchMin=0.8, volumeMax=1}}
 {id=Modulator_1, children [Random_2], properties {pitchMax=1.2, volumeMax=1}}
 {id=QualityLevel_0, children [Random_1,Random_1,Random_0,Random_0]}
 {id=Random_0, children [WavePlayer_0, WavePlayer_9, WavePlayer_14], properties {bShouldExcludeFromBranchCulling=true, weights [1,1,1]}}
 {
  id=Random_1
  children [
   WavePlayer_0
   WavePlayer_1
   WavePlayer_2
   WavePlayer_3
   WavePlayer_9
   WavePlayer_10
   WavePlayer_11
   WavePlayer_12
   WavePlayer_13
   WavePlayer_14
  ]
  properties {bShouldExcludeFromBranchCulling=true, weights [1,1,1,1,1,1,1,1,1,1]}
 }
 {id=Random_2, children [WavePlayer_15, WavePlayer_16, WavePlayer_17, WavePlayer_18], properties {weights [1,1,1,1]}}
 {id=WavePlayer_0, properties {soundWaveAssetPtr=/ProtoSwarm/Assets/Audio/TidalCrane_Swarmer/SoundWaves/Swarmer_Death_01}}
 {id=WavePlayer_1, properties {soundWaveAssetPtr=/ProtoSwarm/Assets/Audio/TidalCrane_Swarmer/SoundWaves/Swarmer_Death_02}}
 {id=WavePlayer_10, properties {soundWaveAssetPtr=/ProtoSwarm/Assets/Audio/TidalCrane_Swarmer/SoundWaves/Swarmer_Death_06}}
 {id=WavePlayer_11, properties {soundWaveAssetPtr=/ProtoSwarm/Assets/Audio/TidalCrane_Swarmer/SoundWaves/Swarmer_Death_07}}
 {id=WavePlayer_12, properties {soundWaveAssetPtr=/ProtoSwarm/Assets/Audio/TidalCrane_Swarmer/SoundWaves/Swarmer_Death_08}}
 {id=WavePlayer_13, properties {soundWaveAssetPtr=/ProtoSwarm/Assets/Audio/TidalCrane_Swarmer/SoundWaves/Swarmer_Death_09}}
 {id=WavePlayer_14, properties {soundWaveAssetPtr=/ProtoSwarm/Assets/Audio/TidalCrane_Swarmer/SoundWaves/Swarmer_Death_10}}
 {id=WavePlayer_15, properties {soundWaveAssetPtr=/ProtoSwarm/Assets/Audio/SoundWaves/Proto_Bug_Crystal_Death_Burst_01}}
 {id=WavePlayer_16, properties {soundWaveAssetPtr=/ProtoSwarm/Assets/Audio/SoundWaves/Proto_Bug_Crystal_Death_Burst_02}}
 {id=WavePlayer_17, properties {soundWaveAssetPtr=/ProtoSwarm/Assets/Audio/SoundWaves/Proto_Bug_Crystal_Death_Burst_03}}
 {id=WavePlayer_18, properties {soundWaveAssetPtr=/ProtoSwarm/Assets/Audio/SoundWaves/Proto_Bug_Crystal_Death_Burst_04}}
 {id=WavePlayer_2, properties {soundWaveAssetPtr=/ProtoSwarm/Assets/Audio/TidalCrane_Swarmer/SoundWaves/Swarmer_Death_03}}
 {id=WavePlayer_3, properties {soundWaveAssetPtr=/ProtoSwarm/Assets/Audio/TidalCrane_Swarmer/SoundWaves/Swarmer_Death_04}}
 {id=WavePlayer_9, properties {soundWaveAssetPtr=/ProtoSwarm/Assets/Audio/TidalCrane_Swarmer/SoundWaves/Swarmer_Death_05}}
]
properties {
 attenuationSettings=/ProtoSwarm/Assets/Audio/Settings/ATT_Proto_Bug_Death_Burst
 bOverrideConcurrency=true
 maxDistance=16750
 soundClassObject=/Game/Sounds/FortSoundClassesAndModes/Fort_Creatures
 volumeMultiplier=1
 concurrencyOverrides {maxCount=4}
 concurrencySet [/ProtoSwarm/Assets/Audio/Settings/CONC_Proto_Bug_Death_Burst_Global, /ProtoSwarm/Assets/Audio/Settings/CONC_Proto_Bug_Death_Burst_Local]
}
