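# Sound cue for bug Swarmer footsteps, mixing a footstep layer with a much quieter body movement layer.
# The footstep layer randomizes pitch, while the movement layer plays at a fixed lowered pitch.
# Quality Level nodes give each layer its full random wave pool in the first two slots and a smaller subset in the last two.
# It uses the Swarmer movement attenuation, the creatures sound class and a concurrency override.
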SoundCue /ProtoSwarm/Assets/Audio/TidalCrane_Swarmer/SC_Bug_Swarmer_Footsteps
root=Mixer_0
nodes [
 {id=Mixer_0, children [Modulator_0,Modulator_1,null], properties {inputVolume [1,0.1,1]}}
 {id=Modulator_0, children [QualityLevel_0], properties {pitchMax=1, pitchMin=0.5, volumeMax=1}}
 {id=Modulator_1, children [QualityLevel_1], properties {pitchMax=0.8, pitchMin=0.8, volumeMax=1, volumeMin=1}}
 {id=QualityLevel_0, children [Random_1,Random_1,Random_0,Random_0]}
 {id=QualityLevel_1, children [Random_2,Random_2,Random_3,Random_3]}
 {
  id=Random_0
  children [WavePlayer_1, WavePlayer_9, WavePlayer_12, WavePlayer_16]
  properties {bShouldExcludeFromBranchCulling=true, weights [1,1,1,1]}
 }
 {
  id=Random_1
  children [
   WavePlayer_0
   WavePlayer_1
   WavePlayer_2
   WavePlayer_3
   WavePlayer_9
   WavePlayer_10
   WavePlayer_11
   WavePlayer_12
   WavePlayer_13
   WavePlayer_14
   WavePlayer_15
   WavePlayer_16
  ]
  properties {bShouldExcludeFromBranchCulling=true, weights [1,1,1,1,1,1,1,1,1,1,1,1]}
 }
 {
  id=Random_2
  children [WavePlayer_17, WavePlayer_18, WavePlayer_19, WavePlayer_20, WavePlayer_21, WavePlayer_22]
  properties {weights [1,1,1,1,1,1]}
 }
 {id=Random_3, children [WavePlayer_17, WavePlayer_19, WavePlayer_22], properties {weights [1,1,1]}}
 {id=WavePlayer_0, properties {soundWaveAssetPtr=/ProtoSwarm/Assets/Audio/TidalCrane_Swarmer/SoundWaves/Swarmer_FootStep_01}}
 {id=WavePlayer_1, properties {soundWaveAssetPtr=/ProtoSwarm/Assets/Audio/TidalCrane_Swarmer/SoundWaves/Swarmer_FootStep_02}}
 {id=WavePlayer_10, properties {soundWaveAssetPtr=/ProtoSwarm/Assets/Audio/TidalCrane_Swarmer/SoundWaves/Swarmer_FootStep_06}}
 {id=WavePlayer_11, properties {soundWaveAssetPtr=/ProtoSwarm/Assets/Audio/TidalCrane_Swarmer/SoundWaves/Swarmer_FootStep_07}}
 {id=WavePlayer_12, properties {soundWaveAssetPtr=/ProtoSwarm/Assets/Audio/TidalCrane_Swarmer/SoundWaves/Swarmer_FootStep_08}}
 {id=WavePlayer_13, properties {soundWaveAssetPtr=/ProtoSwarm/Assets/Audio/TidalCrane_Swarmer/SoundWaves/Swarmer_FootStep_09}}
 {id=WavePlayer_14, properties {soundWaveAssetPtr=/ProtoSwarm/Assets/Audio/TidalCrane_Swarmer/SoundWaves/Swarmer_FootStep_10}}
 {id=WavePlayer_15, properties {soundWaveAssetPtr=/ProtoSwarm/Assets/Audio/TidalCrane_Swarmer/SoundWaves/Swarmer_FootStep_11}}
 {id=WavePlayer_16, properties {soundWaveAssetPtr=/ProtoSwarm/Assets/Audio/TidalCrane_Swarmer/SoundWaves/Swarmer_FootStep_12}}
 {id=WavePlayer_17, properties {soundWaveAssetPtr=/ProtoSwarm/Assets/Audio/TidalCrane_Swarmer/SoundWaves/Swarmer_Movement_01}}
 {id=WavePlayer_18, properties {soundWaveAssetPtr=/ProtoSwarm/Assets/Audio/TidalCrane_Swarmer/SoundWaves/Swarmer_Movement_02}}
 {id=WavePlayer_19, properties {soundWaveAssetPtr=/ProtoSwarm/Assets/Audio/TidalCrane_Swarmer/SoundWaves/Swarmer_Movement_03}}
 {id=WavePlayer_2, properties {soundWaveAssetPtr=/ProtoSwarm/Assets/Audio/TidalCrane_Swarmer/SoundWaves/Swarmer_FootStep_03}}
 {id=WavePlayer_20, properties {soundWaveAssetPtr=/ProtoSwarm/Assets/Audio/TidalCrane_Swarmer/SoundWaves/Swarmer_Movement_04}}
 {id=WavePlayer_21, properties {soundWaveAssetPtr=/ProtoSwarm/Assets/Audio/TidalCrane_Swarmer/SoundWaves/Swarmer_Movement_05}}
 {id=WavePlayer_22, properties {soundWaveAssetPtr=/ProtoSwarm/Assets/Audio/TidalCrane_Swarmer/SoundWaves/Swarmer_Movement_06}}
 {id=WavePlayer_3, properties {soundWaveAssetPtr=/ProtoSwarm/Assets/Audio/TidalCrane_Swarmer/SoundWaves/Swarmer_FootStep_04}}
 {id=WavePlayer_9, properties {soundWaveAssetPtr=/ProtoSwarm/Assets/Audio/TidalCrane_Swarmer/SoundWaves/Swarmer_FootStep_05}}
]
properties {
 attenuationSettings=&/ATT_TidalCrane_Movement_Swarmer
 bOverrideConcurrency=true
 maxDistance=15100
 soundClassObject=/Game/Sounds/FortSoundClassesAndModes/Fort_Creatures
 volumeMultiplier=1
}
