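# Sound cue for swarmer bug flesh impacts, reusing player flesh impact recordings at reduced volume.
# A fixed short delay feeds an equally weighted random pick among four wave variations.
# It uses the pickaxe player impact attenuation asset and a player impact sound class.
# A concurrency override caps instances per owner at three, stopping the oldest, in place of the assigned one-maximum concurrency set.
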SoundCue /ProtoSwarm/Assets/Audio/TidalCrane_Swarmer/SC_Bug_Swarmer_Impact_Flesh_Cue
root=Delay_1
nodes [
 {id=Delay_1, children [Random_1], properties {delayMax=0.1, delayMin=0.1}}
 {id=Random_1, children [WavePlayer_4,WavePlayer_5,WavePlayer_6,WavePlayer_7], properties {weights [1,1,1,1]}}
 {id=WavePlayer_4, properties {soundWaveAssetPtr=/Game/Sounds/Fort_Impact_Sounds/Flesh/new_player_impact_nonspat_A_01n}}
 {id=WavePlayer_5, properties {soundWaveAssetPtr=/Game/Sounds/Fort_Impact_Sounds/Flesh/new_player_impact_nonspat_A_02n}}
 {id=WavePlayer_6, properties {soundWaveAssetPtr=/Game/Sounds/Fort_Impact_Sounds/Flesh/new_player_impact_nonspat_A_03n}}
 {id=WavePlayer_7, properties {soundWaveAssetPtr=/Game/Sounds/Fort_Impact_Sounds/Flesh/new_player_impact_nonspat_A_04n}}
]
properties {
 attenuationSettings=/Game/Sounds/Attenuation/Fort_Weapon_Pickaxe_Player_Impact_Att
 bOverrideConcurrency=true
 maxDistance=3512
 soundClassObject=/Game/Sounds/FortSoundClassesAndModes/Fort_Impact_Player_PSM
 volumeMultiplier=0.3
 attenuationOverrides {
  bAttenuate=false
  bAttenuateWithLPF=true
  distanceAlgorithm=Logarithmic
  falloffDistance=800
  lPFRadiusMax=800
  lPFRadiusMin=400
  attenuationShapeExtents {x=300}
 }
 concurrencyOverrides {bLimitToOwner=true, maxCount=3, resolutionRule=StopOldest}
 concurrencySet [/Game/Sounds/Concurrency_Rules/OneMaximum_Owner_StopOldest]
}
