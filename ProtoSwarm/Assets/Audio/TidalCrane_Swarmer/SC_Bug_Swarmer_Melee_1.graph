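# Sound cue for the Swarmer bug's first melee attack, played in the creatures sound class with melee attenuation and shared Swarmer voice concurrency.
# A quality-level switch sends its first two levels to a 25-wave random pool of standard, bite and no-vocal takes.
# The other two levels use a three-wave subset, one of each type, reusing the same players.
# A modulator boosts volume on top of the cue's raised multiplier.
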SoundCue /ProtoSwarm/Assets/Audio/TidalCrane_Swarmer/SC_Bug_Swarmer_Melee_1
root=Modulator_0
nodes [
 {id=Modulator_0, children [QualityLevel_0], properties {volumeMax=2, volumeMin=2}}
 {id=QualityLevel_0, children [Random_1,Random_1,Random_0,Random_0]}
 {id=Random_0, children [WavePlayer_1, WavePlayer_12, WavePlayer_23], properties {bShouldExcludeFromBranchCulling=true, weights [1,1,1]}}
 {
  id=Random_1
  children [
   WavePlayer_0
   WavePlayer_1
   WavePlayer_2
   WavePlayer_3
   WavePlayer_4
   WavePlayer_5
   WavePlayer_6
   WavePlayer_7
   WavePlayer_8
   WavePlayer_9
   WavePlayer_10
   WavePlayer_11
   WavePlayer_12
   WavePlayer_13
   WavePlayer_14
   WavePlayer_15
   WavePlayer_16
   WavePlayer_17
   WavePlayer_18
   WavePlayer_19
   WavePlayer_20
   WavePlayer_21
   WavePlayer_22
   WavePlayer_23
   WavePlayer_24
  ]
  properties {bShouldExcludeFromBranchCulling=true, weights [1,1,1,1,1,1,1,1,1,1,1,1,1,1,1,1,1,1,1,1,1,1,1,1,1]}
 }
 {id=WavePlayer_0, properties {soundWaveAssetPtr=/ProtoSwarm/Assets/Audio/TidalCrane_Swarmer/SoundWaves/Swarmer_Melee_1_01}}
 {id=WavePlayer_1, properties {soundWaveAssetPtr=/ProtoSwarm/Assets/Audio/TidalCrane_Swarmer/SoundWaves/Swarmer_Melee_1_02}}
 {id=WavePlayer_10, properties {soundWaveAssetPtr=/ProtoSwarm/Assets/Audio/TidalCrane_Swarmer/SoundWaves/Swarmer_Melee_1_Bite_01}}
 {id=WavePlayer_11, properties {soundWaveAssetPtr=/ProtoSwarm/Assets/Audio/TidalCrane_Swarmer/SoundWaves/Swarmer_Melee_1_Bite_02}}
 {id=WavePlayer_12, properties {soundWaveAssetPtr=/ProtoSwarm/Assets/Audio/TidalCrane_Swarmer/SoundWaves/Swarmer_Melee_1_Bite_03}}
 {id=WavePlayer_13, properties {soundWaveAssetPtr=/ProtoSwarm/Assets/Audio/TidalCrane_Swarmer/SoundWaves/Swarmer_Melee_1_Bite_04}}
 {id=WavePlayer_14, properties {soundWaveAssetPtr=/ProtoSwarm/Assets/Audio/TidalCrane_Swarmer/SoundWaves/Swarmer_Melee_1_Bite_05}}
 {id=WavePlayer_15, properties {soundWaveAssetPtr=/ProtoSwarm/Assets/Audio/TidalCrane_Swarmer/SoundWaves/Swarmer_Melee_1_11}}
 {id=WavePlayer_16, properties {soundWaveAssetPtr=/ProtoSwarm/Assets/Audio/TidalCrane_Swarmer/SoundWaves/Swarmer_Melee_1_12}}
 {id=WavePlayer_17, properties {soundWaveAssetPtr=/ProtoSwarm/Assets/Audio/TidalCrane_Swarmer/SoundWaves/Swarmer_Melee_1_13}}
 {id=WavePlayer_18, properties {soundWaveAssetPtr=/ProtoSwarm/Assets/Audio/TidalCrane_Swarmer/SoundWaves/Swarmer_Melee_1_14}}
 {id=WavePlayer_19, properties {soundWaveAssetPtr=/ProtoSwarm/Assets/Audio/TidalCrane_Swarmer/SoundWaves/Swarmer_Melee_1_15}}
 {id=WavePlayer_2, properties {soundWaveAssetPtr=/ProtoSwarm/Assets/Audio/TidalCrane_Swarmer/SoundWaves/Swarmer_Melee_1_03}}
 {id=WavePlayer_20, properties {soundWaveAssetPtr=/ProtoSwarm/Assets/Audio/TidalCrane_Swarmer/SoundWaves/Swarmer_Melee_1_NoVox_01}}
 {id=WavePlayer_21, properties {soundWaveAssetPtr=/ProtoSwarm/Assets/Audio/TidalCrane_Swarmer/SoundWaves/Swarmer_Melee_1_NoVox_05}}
 {id=WavePlayer_22, properties {soundWaveAssetPtr=/ProtoSwarm/Assets/Audio/TidalCrane_Swarmer/SoundWaves/Swarmer_Melee_1_NoVox_04}}
 {id=WavePlayer_23, properties {soundWaveAssetPtr=/ProtoSwarm/Assets/Audio/TidalCrane_Swarmer/SoundWaves/Swarmer_Melee_1_NoVox_02}}
 {id=WavePlayer_24, properties {soundWaveAssetPtr=/ProtoSwarm/Assets/Audio/TidalCrane_Swarmer/SoundWaves/Swarmer_Melee_1_NoVox_03}}
 {id=WavePlayer_3, properties {soundWaveAssetPtr=/ProtoSwarm/Assets/Audio/TidalCrane_Swarmer/SoundWaves/Swarmer_Melee_1_04}}
 {id=WavePlayer_4, properties {soundWaveAssetPtr=/ProtoSwarm/Assets/Audio/TidalCrane_Swarmer/SoundWaves/Swarmer_Melee_1_05}}
 {id=WavePlayer_5, properties {soundWaveAssetPtr=/ProtoSwarm/Assets/Audio/TidalCrane_Swarmer/SoundWaves/Swarmer_Melee_1_06}}
 {id=WavePlayer_6, properties {soundWaveAssetPtr=/ProtoSwarm/Assets/Audio/TidalCrane_Swarmer/SoundWaves/Swarmer_Melee_1_07}}
 {id=WavePlayer_7, properties {soundWaveAssetPtr=/ProtoSwarm/Assets/Audio/TidalCrane_Swarmer/SoundWaves/Swarmer_Melee_1_08}}
 {id=WavePlayer_8, properties {soundWaveAssetPtr=/ProtoSwarm/Assets/Audio/TidalCrane_Swarmer/SoundWaves/Swarmer_Melee_1_09}}
 {id=WavePlayer_9, properties {soundWaveAssetPtr=/ProtoSwarm/Assets/Audio/TidalCrane_Swarmer/SoundWaves/Swarmer_Melee_1_10}}
]
properties {
 attenuationSettings=/ProtoSwarm/Assets/Audio/Settings/ATT_Proto_Bug_Melee
 maxDistance=10750
 priority=5
 soundClassObject=/Game/Sounds/FortSoundClassesAndModes/Fort_Creatures
 volumeMultiplier=2
 concurrencyOverrides {maxCount=4}
 concurrencySet [/ProtoSwarm/Assets/Audio/Settings/CONC_TidalCrane_Swarmer_VO_Shared]
}
