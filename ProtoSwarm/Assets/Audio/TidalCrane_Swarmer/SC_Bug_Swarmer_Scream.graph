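# Sound cue for the swarmer bug's scream, played as a creature-class sound with melee-bug attenuation.
# A quality-level node sends its first two levels to a random pick of five scream waves and its last two to a three-wave subset.
# A root modulator keeps volume fixed while leaving default pitch variation.
# A shared swarmer voice concurrency set limits overlapping screams.
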SoundCue /ProtoSwarm/Assets/Audio/TidalCrane_Swarmer/SC_Bug_Swarmer_Scream
root=Modulator_0
nodes [
 {id=Modulator_0, children [QualityLevel_1], properties {volumeMax=1, volumeMin=1}}
 {id=QualityLevel_1, children [Random_2,Random_2,Random_3,Random_3]}
 {id=Random_2, children [WavePlayer_25, WavePlayer_26, WavePlayer_27, WavePlayer_28, WavePlayer_29], properties {weights [1,1,1,1,1]}}
 {id=Random_3, children [WavePlayer_25, WavePlayer_27, WavePlayer_29], properties {bShouldExcludeFromBranchCulling=true, weights [1,1,1]}}
 {id=WavePlayer_25, properties {soundWaveAssetPtr=/ProtoSwarm/Assets/Audio/TidalCrane_Swarmer/SoundWaves/Swarmer_Scream_01}}
 {id=WavePlayer_26, properties {soundWaveAssetPtr=/ProtoSwarm/Assets/Audio/TidalCrane_Swarmer/SoundWaves/Swarmer_Scream_02}}
 {id=WavePlayer_27, properties {soundWaveAssetPtr=/ProtoSwarm/Assets/Audio/TidalCrane_Swarmer/SoundWaves/Swarmer_Scream_03}}
 {id=WavePlayer_28, properties {soundWaveAssetPtr=/ProtoSwarm/Assets/Audio/TidalCrane_Swarmer/SoundWaves/Swarmer_Scream_04}}
 {id=WavePlayer_29, properties {soundWaveAssetPtr=/ProtoSwarm/Assets/Audio/TidalCrane_Swarmer/SoundWaves/Swarmer_Scream_05}}
]
properties {
 attenuationSettings=/ProtoSwarm/Assets/Audio/Settings/ATT_Proto_Bug_Melee
 maxDistance=10750
 priority=5
 soundClassObject=/Game/Sounds/FortSoundClassesAndModes/Fort_Creatures
 volumeMultiplier=2
 concurrencyOverrides {maxCount=4}
 concurrencySet [/ProtoSwarm/Assets/Audio/Settings/CONC_TidalCrane_Swarmer_VO_Shared]
}
