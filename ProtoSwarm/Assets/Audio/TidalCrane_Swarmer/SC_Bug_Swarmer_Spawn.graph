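# Sound cue that plays the spawn sound for the Swarmer bug creature, routed through the Fort_Creatures sound class.
# A Random node picks one of four equally weighted spawn waves, then a Modulator adds variation at playback.
# It uses a dedicated spawn attenuation asset and a shared Swarmer VO concurrency set to manage overlapping instances.
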SoundCue /ProtoSwarm/Assets/Audio/TidalCrane_Swarmer/SC_Bug_Swarmer_Spawn
root=Modulator_0
nodes [
 {id=Modulator_0, children [Random_1], properties {volumeMax=1}}
 {
  id=Random_1
  children [WavePlayer_4,WavePlayer_5,WavePlayer_6,WavePlayer_7]
  properties {bShouldExcludeFromBranchCulling=true, weights [1,1,1,1]}
 }
 {id=WavePlayer_4, properties {soundWaveAssetPtr=/ProtoSwarm/Assets/Audio/TidalCrane_Swarmer/SoundWaves/Swarmer_Spawn_01}}
 {id=WavePlayer_5, properties {soundWaveAssetPtr=/ProtoSwarm/Assets/Audio/TidalCrane_Swarmer/SoundWaves/Swarmer_Spawn_02}}
 {id=WavePlayer_6, properties {soundWaveAssetPtr=/ProtoSwarm/Assets/Audio/TidalCrane_Swarmer/SoundWaves/Swarmer_Spawn_03}}
 {id=WavePlayer_7, properties {soundWaveAssetPtr=/ProtoSwarm/Assets/Audio/TidalCrane_Swarmer/SoundWaves/Swarmer_Spawn_04}}
]
properties {
 attenuationSettings=&/ATT_TidalCrane_Swarmer_Spawn
 maxDistance=16000
 priority=5
 soundClassObject=/Game/Sounds/FortSoundClassesAndModes/Fort_Creatures
 volumeMultiplier=1.6
 concurrencyOverrides {maxCount=4}
 concurrencySet [/ProtoSwarm/Assets/Audio/Settings/CONC_TidalCrane_Swarmer_VO_Shared]
}
